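# Animation Blueprint for the ChimeCurl microphone instrument cosmetic that adds physics swing to its dangling chain and star charm.
# The input pose runs through five serial Anim Dynamics nodes in component space, one each for three chain bones, the star ring and the star.
# Each node uses overridden gravity, high angular damping, scaled, clamped component motion and per-bone angular limits.
# The EventGraph has only unconnected stub nodes, with no logic or variables.
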AnimBlueprint /SparksCosmetics/Instrument/Mic/ChimeCurl/Meshes/ChimeCurl_Mic_AnimBP
target_skeleton=&/ChimeCurl_Mic_Skeleton
graphs [
 AnimGraph {nodes [
   n1 = Anim Dynamics - Bone: dyn_mic_chain_01!(ComponentPose=Local To Component(LocalPose=Input Pose()), bAlphaBoolEnabled:bool=True, ExternalForce="0,0,0") -> (Pose) {
    alphaInputType=Bool
    angularDampingOverride=0.99
    bOverrideAngularDamping=true
    bUseGravityOverride=true
    alphaBoolBlend {blendOption=Linear}
    boundBone {boneName=dyn_mic_chain_01}
    componentAppliedLinearAccClamp=<200 200 200>
    componentLinearAccScale=<0.2 0.2 0.2>
    componentLinearVelScale=<0.5 0.5 0.5>
    gravityOverride {z=-980}
    physicsBodyDefinitions [{
      boundBone {boneName=dyn_mic_chain_01}
      boxExtents=<10 10 10>
      constraintSetup {angularTargetAxis=AxisY, twistAxis=AxisZ, angularLimitsMax=<0 45.03 0>, angularLimitsMin=<0 -30 0>, angularTarget=<0 0 0>}
      localJointOffset=<1 0 0>
     }]
   }
   n2 = Anim Dynamics - Bone: dyn_mic_chain_02!(ComponentPose=@n1, bAlphaBoolEnabled:bool=True, GravityOverride:Vector="0,0,-980", ExternalForce="0,0,0") -> (Pose) {
    alphaInputType=Bool
    angularDampingOverride=0.99
    bOverrideAngularDamping=true
    bUseGravityOverride=true
    linearDampingOverride=0.9
    alphaBoolBlend {blendOption=Linear}
    boundBone {boneName=dyn_mic_chain_02}
    componentAppliedLinearAccClamp=<200 200 200>
    componentLinearAccScale=<0.2 0.2 0.2>
    componentLinearVelScale=<0.5 0.5 0.5>
    physicsBodyDefinitions [{
      boundBone {boneName=dyn_mic_chain_02}
      boxExtents=<6 6 6>
      constraintSetup {angularLimitsMax=<15 20 45>, angularLimitsMin=<-15 -45 -45>, angularTarget=<0 0 0>, linearAxesMin=<-0.25 -0.25 -0.25>}
      localJointOffset=<1 0 0>
     }]
   }
   n3 = Anim Dynamics - Bone: dyn_mic_chain_03!(ComponentPose=@n2, bAlphaBoolEnabled:bool=True, GravityOverride:Vector="0,0,-980", ExternalForce="0,0,0") -> (Pose) {
    alphaInputType=Bool
    angularDampingOverride=0.99
    bOverrideAngularDamping=true
    bUseGravityOverride=true
    linearDampingOverride=0.9
    alphaBoolBlend {blendOption=Linear}
    boundBone {boneName=dyn_mic_chain_03}
    componentAppliedLinearAccClamp=<200 200 200>
    componentLinearAccScale=<0.2 0.2 0.2>
    componentLinearVelScale=<0.5 0.5 0.5>
    physicsBodyDefinitions [{
      boundBone {boneName=dyn_mic_chain_03}
      boxExtents=<6 6 6>
      constraintSetup {angularLimitsMax=<15 35 45>, angularLimitsMin=<-15 -60 -45>, angularTarget=<0 0 0>, linearAxesMin=<-0.25 -0.25 -0.25>}
      localJointOffset=<1 0 0>
     }]
   }
   n4 = Anim Dynamics - Bone: dyn_mic_star_ring!(ComponentPose=@n3, bAlphaBoolEnabled:bool=True, GravityOverride:Vector="0,0,-980", ExternalForce="0,0,0") -> (Pose) {
    alphaInputType=Bool
    angularDampingOverride=0.99
    bOverrideAngularDamping=true
    bUseGravityOverride=true
    linearDampingOverride=0.9
    alphaBoolBlend {blendOption=Linear}
    boundBone {boneName=dyn_mic_star_ring}
    componentAppliedLinearAccClamp=<200 200 200>
    componentLinearAccScale=<0.2 0.2 0.2>
    componentLinearVelScale=<0.5 0.5 0.5>
    physicsBodyDefinitions [{
      boundBone {boneName=dyn_mic_star_ring}
      boxExtents=<6 6 6>
      constraintSetup {angularLimitsMax=<5 20 45>, angularLimitsMin=<-5 -45 -45>, angularTarget=<0 0 0>, linearAxesMin=<-0.25 -0.25 -0.25>}
      localJointOffset=<1 0 0>
     }]
   }
   n5 = Anim Dynamics - Bone: dyn_mic_star!(ComponentPose=@n4, bAlphaBoolEnabled:bool=True, GravityOverride:Vector="0,0,-980", ExternalForce="0,0,0") -> (Pose) {
    alphaInputType=Bool
    angularDampingOverride=0.99
    bOverrideAngularDamping=true
    bUseGravityOverride=true
    linearDampingOverride=0.9
    alphaBoolBlend {blendOption=Linear}
    boundBone {boneName=dyn_mic_star}
    componentAppliedLinearAccClamp=<200 200 200>
    componentLinearAccScale=<0.2 0.2 0.2>
    componentLinearVelScale=<0.5 0.5 0.5>
    physicsBodyDefinitions [{
      boundBone {boneName=dyn_mic_star}
      boxExtents=<10 10 10>
      constraintSetup {angularLimitsMax=<15 45 20>, angularLimitsMin=<-15 -25 -20>, angularTarget=<0 0 0>, linearAxesMin=<-0.25 -0.25 -0.25>}
      localJointOffset=<4 0 0>
     }]
   }
   Output Pose(Result=Component To Local(ComponentPose=@n5))
  ]}
 EventGraph {nodes [
   TryGetPawnOwner()
   Event BlueprintUpdateAnimation()
  ]}
]
variables []
